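# Physics asset for the base skeleton of the drivable Mech vehicle, giving it a simplified body for collision and physics.
# One capsule per major bone covers pelvis, two spine segments, upper and lower arms, thighs, calves and feet, with mirrored left and right sides.
# Constraints chain the limbs and spine back to the pelvis.
# All swing and twist axes are Free, so the stored 45-degree limits do not restrict rotation.
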PhysicsAsset /Game/Athena/DrivableVehicles/Mech/Base_Skeleton/Mech_Veh_Base_Physics
bodies {
 calf_l [{shapeType=Capsule, length=135.25, radius=38, center=<-21.7896 18.2841 -0.0001>, rotation=<rot 90 10.6197 -39.379>}]
 calf_r [{shapeType=Capsule, length=135.25, radius=38, center=<21.7896 -18.2841 0.0001>, rotation=<rot 90 10.6197 -39.379>}]
 foot_l [{shapeType=Capsule, length=57, radius=50, center=<-12.3315 35.6559 0.003>, rotation=<rot 90 0 89.9997>}]
 foot_r [{shapeType=Capsule, length=57, radius=50, center=<12.3315 -35.6559 -0.003>, rotation=<rot 90 0 89.9997>}]
 lowerarm_l [{shapeType=Capsule, length=216.3525, radius=60.8549, center=<169.781 0.0271 17.1384>, rotation=<rot 0 89.0703 -84.3879>}]
 lowerarm_r [{shapeType=Capsule, length=216.3525, radius=60.8549, center=<-169.781 -0.0271 -17.1384>, rotation=<rot 0 89.0703 -84.3879>}]
 pelvis [{shapeType=Capsule, length=126.6467, radius=57.829}]
 spine_03 [{shapeType=Capsule, length=54.1709, radius=68.8047}]
 spine_05 [{shapeType=Capsule, length=162.5125, radius=78.9247, center=<46.3377 0.0006 -0.0001>}]
 thigh_l [{shapeType=Capsule, length=123.25, radius=50, center=<-44.4428 46.4104 -0.0002>, rotation=<rot 90 0 39.999>}]
 thigh_r [{shapeType=Capsule, length=123.25, radius=50, center=<44.4428 -46.4104 0.0002>, rotation=<rot 90 0 39.999>}]
 upperarm_l [{shapeType=Capsule, length=52.179, radius=37.4107, center=<45.164 0.3886 -0.7208>, rotation=<rot 0 89.5019 -89.6206>}]
 upperarm_r [{shapeType=Capsule, length=52.179, radius=37.4107, center=<-45.164 -0.3886 0.7208>, rotation=<rot 0 89.5019 -89.6206>}]
}
constraints [
 {
  bone1Name=spine_03
  bone2Name=pelvis
  swing1LimitDegrees=45
  swing1Motion=Free
  swing2LimitDegrees=45
  swing2Motion=Free
  twistLimitDegrees=45
  twistMotion=Free
 }
 {
  bone1Name=upperarm_l
  bone2Name=spine_05
  swing1LimitDegrees=45
  swing1Motion=Free
  swing2LimitDegrees=45
  swing2Motion=Free
  twistLimitDegrees=45
  twistMotion=Free
 }
 {
  bone1Name=lowerarm_l
  bone2Name=upperarm_l
  swing1LimitDegrees=45
  swing1Motion=Free
  swing2LimitDegrees=45
  swing2Motion=Free
  twistLimitDegrees=45
  twistMotion=Free
 }
 {
  bone1Name=upperarm_r
  bone2Name=spine_05
  swing1LimitDegrees=45
  swing1Motion=Free
  swing2LimitDegrees=45
  swing2Motion=Free
  twistLimitDegrees=45
  twistMotion=Free
 }
 {
  bone1Name=lowerarm_r
  bone2Name=upperarm_r
  swing1LimitDegrees=45
  swing1Motion=Free
  swing2LimitDegrees=45
  swing2Motion=Free
  twistLimitDegrees=45
  twistMotion=Free
 }
 {
  bone1Name=thigh_l
  bone2Name=pelvis
  swing1LimitDegrees=45
  swing1Motion=Free
  swing2LimitDegrees=45
  swing2Motion=Free
  twistLimitDegrees=45
  twistMotion=Free
 }
 {
  bone1Name=calf_l
  bone2Name=thigh_l
  swing1LimitDegrees=45
  swing1Motion=Free
  swing2LimitDegrees=45
  swing2Motion=Free
  twistLimitDegrees=45
  twistMotion=Free
 }
 {
  bone1Name=foot_l
  bone2Name=calf_l
  swing1LimitDegrees=45
  swing1Motion=Free
  swing2LimitDegrees=45
  swing2Motion=Free
  twistLimitDegrees=45
  twistMotion=Free
 }
 {
  bone1Name=thigh_r
  bone2Name=pelvis
  swing1LimitDegrees=45
  swing1Motion=Free
  swing2LimitDegrees=45
  swing2Motion=Free
  twistLimitDegrees=45
  twistMotion=Free
 }
 {
  bone1Name=calf_r
  bone2Name=thigh_r
  swing1LimitDegrees=45
  swing1Motion=Free
  swing2LimitDegrees=45
  swing2Motion=Free
  twistLimitDegrees=45
  twistMotion=Free
 }
 {
  bone1Name=foot_r
  bone2Name=calf_r
  swing1LimitDegrees=45
  swing1Motion=Free
  swing2LimitDegrees=45
  swing2Motion=Free
  twistLimitDegrees=45
  twistMotion=Free
 }
 {
  bone1Name=spine_05
  bone2Name=spine_03
  swing1LimitDegrees=45
  swing1Motion=Free
  swing2LimitDegrees=45
  swing2Motion=Free
  twistLimitDegrees=45
  twistMotion=Free
 }
]
